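# Sound cue for the copy action of the Creative MoveTool gadget, played spatially with the MoveTool attenuation and the Creative gadgets sound class.
# A mixer layers the copy wave, the place wave at full input volume, and a quieter scan layer.
# The scan layer randomly picks one of two place-scan waves and passes it through a modulator that applies a fixed pitch raise.
# The whole cue is also pitched up and reduced in volume.
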SoundCue /Game/Sounds/Creative/Gadgets/MoveTool/MoveTool_Copy_01_Cue
root=Mixer_0
nodes [
 {id=Mixer_0, children [WavePlayer_0,WavePlayer_1,Modulator_0], properties {inputVolume [0.6,1,0.3]}}
 {id=Modulator_0, children [Random_0], properties {pitchMax=1.2, pitchMin=1.2}}
 {id=Random_0, children [WavePlayer_3,WavePlayer_4], properties {weights [1,1]}}
 {id=WavePlayer_0, properties {soundWaveAssetPtr=&/MoveTool_Copy_01}}
 {id=WavePlayer_1, properties {soundWaveAssetPtr=&/MoveTool_Place_01}}
 {id=WavePlayer_3, properties {soundWaveAssetPtr=&/MoveTool_Place_Scan_01}}
 {id=WavePlayer_4, properties {soundWaveAssetPtr=&/MoveTool_Place_Scan_02}}
]
properties {
 attenuationSettings=/Game/Sounds/Creative/Attenuations/Fort_Gadget_MoveTool_Att
 maxDistance=4000
 pitchMultiplier=1.4
 soundClassObject=&Fort_Gadgets_Creative
 volumeMultiplier=0.55
}
