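# MoveTool_Delete_01_Cue is a sound cue for the delete action of the Creative move tool gadget.
# A Mixer layers two waves at once: Delete_01 through a Modulator at reduced input volume, and Delete_02 at full input volume.
# The cue plays in the Creative gadgets sound class with lowered overall volume and the move tool's attenuation asset.
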SoundCue /Game/Sounds/Creative/Gadgets/MoveTool/MoveTool_Delete_01_Cue
root=Mixer_0
nodes [
 {id=Mixer_0, children [Modulator_0,WavePlayer_3], properties {inputVolume [0.7,1]}}
 {id=Modulator_0, children [WavePlayer_0]}
 {id=WavePlayer_0, properties {soundWaveAssetPtr=&/MoveTool_Delete_01}}
 {id=WavePlayer_3, properties {soundWaveAssetPtr=&/MoveTool_Delete_02}}
]
properties {
 attenuationSettings=/Game/Sounds/Creative/Attenuations/Fort_Gadget_MoveTool_Att
 maxDistance=4000
 soundClassObject=&Fort_Gadgets_Creative
 volumeMultiplier=0.45
}
