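# MoveTool_Equip_Cue is a sound cue for the Creative Move Tool gadget, named for its equip sound.
# A single wave player feeds the output directly with the MoveTool_Axis_01 wave at reduced volume, without randomization or modulation.
# It routes through the Creative gadgets sound class, uses the Move Tool attenuation asset, and overrides concurrency to cap simultaneous plays, stopping the oldest.
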SoundCue /Game/Sounds/Creative/Gadgets/MoveTool/MoveTool_Equip_Cue
root=WavePlayer_1
nodes [{id=WavePlayer_1, properties {soundWaveAssetPtr=&/MoveTool_Axis_01}}]
properties {
 attenuationSettings=/Game/Sounds/Creative/Attenuations/Fort_Gadget_MoveTool_Att
 bOverrideConcurrency=true
 maxDistance=4000
 soundClassObject=&Fort_Gadgets_Creative
 volumeMultiplier=0.5
 concurrencyOverrides {maxCount=2, resolutionRule=StopOldest}
}
