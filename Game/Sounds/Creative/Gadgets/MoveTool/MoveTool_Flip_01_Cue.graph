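# Sound cue for the flip sound of the Creative mode Move Tool gadget.
# A single wave player plays the MoveTool_Flip_01 wave, with the cue doubling its pitch and lowering its volume.
# It uses the Move Tool attenuation asset and routes through the Creative gadgets sound class.
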SoundCue /Game/Sounds/Creative/Gadgets/MoveTool/MoveTool_Flip_01_Cue
root=WavePlayer_0
nodes [{id=WavePlayer_0, properties {soundWaveAssetPtr=&/MoveTool_Flip_01}}]
properties {
 attenuationSettings=/Game/Sounds/Creative/Attenuations/Fort_Gadget_MoveTool_Att
 maxDistance=4000
 pitchMultiplier=2
 soundClassObject=&Fort_Gadgets_Creative
 volumeMultiplier=0.4
}
