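# Looping sound cue for the Creative move tool's flight, reusing a falling-wind loop from Athena.
# A continuous modulator drives both pitch and volume of the looping wave player from a runtime "Speed" parameter, each with its own clamped range.
# The cue plays in the Creative gadgets sound class with overridden attenuation and a concurrency limit of a single instance.
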SoundCue /Game/Sounds/Creative/Gadgets/MoveTool/MoveTool_FlyLoop_01_Cue
root=ModulatorContinuous_0
nodes [
 {
  id=ModulatorContinuous_0
  children [WavePlayer_3]
  properties {
   pitchModulationParams {maxInput=0.9, maxOutput=0.9, minInput=0.5, minOutput=0.5, parameterName=Speed}
   volumeModulationParams {default=0.5, minInput=0.3, minOutput=0.3, parameterName=Speed}
  }
 }
 {id=WavePlayer_3, properties {bLooping=true, soundWaveAssetPtr=/Game/Athena/Sounds/Wind_Falling/Wind_UmbrellaClosed_Loop_01}}
]
properties {
 bOverrideAttenuation=true
 bOverrideConcurrency=true
 maxDistance=1400
 soundClassObject=&Fort_Gadgets_Creative
 volumeMultiplier=0.5
 attenuationOverrides {falloffDistance=1000, stereoSpread=800}
 concurrencyOverrides {maxCount=1}
}
