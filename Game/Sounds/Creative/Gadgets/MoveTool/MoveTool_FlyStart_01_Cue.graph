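# Sound cue for the Creative MoveTool gadget's fly-start sound, routed to the Creative gadgets sound class.
# A Random node picks one of three equally weighted wave variations and feeds a Modulator that varies playback, with pitch capped at its original level.
# It uses the MoveTool attenuation asset at reduced volume, and a concurrency override allows one instance at a time, blocking new plays.
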SoundCue /Game/Sounds/Creative/Gadgets/MoveTool/MoveTool_FlyStart_01_Cue
root=Modulator_0
nodes [
 {id=Modulator_0, children [Random_0], properties {pitchMax=1}}
 {id=Random_0, children [WavePlayer_1,WavePlayer_3,WavePlayer_2], properties {weights [1,1,1]}}
 {id=WavePlayer_1, properties {soundWaveAssetPtr=&/MoveTool_FlyStart_02}}
 {id=WavePlayer_2, properties {soundWaveAssetPtr=&/MoveTool_FlyStart_04}}
 {id=WavePlayer_3, properties {soundWaveAssetPtr=&/MoveTool_FlyStart_03}}
]
properties {
 attenuationSettings=/Game/Sounds/Creative/Attenuations/Fort_Gadget_MoveTool_Att
 bOverrideConcurrency=true
 maxDistance=4000
 soundClassObject=&Fort_Gadgets_Creative
 volumeMultiplier=0.4
 concurrencyOverrides {maxCount=1, resolutionRule=PreventNew}
}
