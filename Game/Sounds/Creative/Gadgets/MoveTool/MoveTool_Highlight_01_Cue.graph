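# Sound cue for the Creative move tool gadget's highlight sound.
# A single wave player plays one highlight sample at reduced volume, routed to the Creative gadgets sound class.
# A dedicated move tool attenuation asset handles spatial falloff.
# Overridden concurrency limits instances per owner to a small count, blocks new plays beyond it and applies volume scaling to concurrent instances.
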SoundCue /Game/Sounds/Creative/Gadgets/MoveTool/MoveTool_Highlight_01_Cue
root=WavePlayer_0
nodes [{id=WavePlayer_0, properties {soundWaveAssetPtr=&/MoveTool_Highlight_01}}]
properties {
 attenuationSettings=/Game/Sounds/Creative/Attenuations/Fort_Gadget_MoveTool_Att
 bOverrideConcurrency=true
 maxDistance=4000
 soundClassObject=&Fort_Gadgets_Creative
 volumeMultiplier=0.25
 concurrencyOverrides {bLimitToOwner=true, maxCount=2, resolutionRule=PreventNew, volumeScale=0.5}
}
