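# Sound cue for the Creative move tool gadget, named for the multi-select duplicate action.
# A single wave player at the root plays the MoveTool_Granted_02 wave directly, with no randomization, mixing or modulation.
# It uses the move tool attenuation asset, belongs to the Creative gadgets sound class, and lowers the volume below full.
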SoundCue /Game/Sounds/Creative/Gadgets/MoveTool/MoveTool_MultiSelect_Duplicate_Cue
root=WavePlayer_1
nodes [{id=WavePlayer_1, properties {soundWaveAssetPtr=&/MoveTool_Granted_02}}]
properties {
 attenuationSettings=/Game/Sounds/Creative/Attenuations/Fort_Gadget_MoveTool_Att
 maxDistance=4000
 soundClassObject=&Fort_Gadgets_Creative
 volumeMultiplier=0.65
}
