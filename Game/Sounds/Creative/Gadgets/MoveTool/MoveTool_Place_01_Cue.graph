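# Sound cue for the Creative move tool gadget, named for its place action, routed to the Creative gadgets sound class.
# A single wave player reuses the MoveTool_Select_01 wave, with the cue raising its pitch and lowering its volume.
# It uses the move tool attenuation asset and overrides concurrency to allow one instance at a time, stopping the oldest.
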SoundCue /Game/Sounds/Creative/Gadgets/MoveTool/MoveTool_Place_01_Cue
root=WavePlayer_0
nodes [{id=WavePlayer_0, properties {soundWaveAssetPtr=&/MoveTool_Select_01}}]
properties {
 attenuationSettings=/Game/Sounds/Creative/Attenuations/Fort_Gadget_MoveTool_Att
 bOverrideConcurrency=true
 maxDistance=4000
 pitchMultiplier=1.5
 soundClassObject=&Fort_Gadgets_Creative
 volumeMultiplier=0.3
 concurrencyOverrides {bVolumeScaleCanRelease=true, maxCount=1, resolutionRule=StopOldest, volumeScaleAttackTime=0.1, volumeScaleReleaseTime=0.3}
}
