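# MoveTool_Raise_01_Cue is a Sound Cue for a raise sound of the Creative MoveTool gadget.
# A single wave player feeds the output directly with MoveTool_Raise_03, despite the cue's 01 name, at reduced volume.
# It uses the Creative gadgets sound class and MoveTool attenuation, with a concurrency override that caps simultaneous instances by stopping the oldest.
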SoundCue /Game/Sounds/Creative/Gadgets/MoveTool/MoveTool_Raise_01_Cue
root=WavePlayer_0
nodes [{id=WavePlayer_0, properties {soundWaveAssetPtr=&/MoveTool_Raise_03}}]
properties {
 attenuationSettings=/Game/Sounds/Creative/Attenuations/Fort_Gadget_MoveTool_Att
 bOverrideConcurrency=true
 maxDistance=4000
 soundClassObject=&Fort_Gadgets_Creative
 volumeMultiplier=0.7
 concurrencyOverrides {maxCount=2, resolutionRule=StopOldest}
}
