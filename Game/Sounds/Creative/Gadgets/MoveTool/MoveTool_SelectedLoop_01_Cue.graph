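# Sound cue for the selected-loop sound of the Creative move tool gadget, going by its name, in the creative gadgets sound class with move tool attenuation.
# A single wave player feeds MoveTool_Place_Scan_02 through a modulator that applies a fixed upward pitch shift, then into a mixer at reduced volume.
# The mixer's first input is empty, and no looping flag is set on the wave player.
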SoundCue /Game/Sounds/Creative/Gadgets/MoveTool/MoveTool_SelectedLoop_01_Cue
root=Mixer_0
nodes [
 {id=Mixer_0, children [null,Modulator_0], properties {inputVolume [1,0.4]}}
 {id=Modulator_0, children [WavePlayer_4], properties {pitchMax=1.6, pitchMin=1.6}}
 {id=WavePlayer_4, properties {soundWaveAssetPtr=&/MoveTool_Place_Scan_02}}
]
properties {
 attenuationSettings=/Game/Sounds/Creative/Attenuations/Fort_Gadget_MoveTool_Att
 maxDistance=4000
 soundClassObject=&Fort_Gadgets_Creative
 volumeMultiplier=0.3
}
